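# InsectSwarm death SoundCue, layering a random death sound with a random sting hiss for variation each play.
# A Mixer blends one of three death waves with one of four hiss waves, the hiss passing through a Modulator that raises pitch and slightly varies volume.
# It uses close-range enemy attenuation and a concurrency override that stops the farthest instance, then prevents new ones.
# All waves come from the ChronoSwarmGrenade content.
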SoundCue /InsectSwarm/Sounds/Assets/InsectSwarm_Death_Cue
root=Mixer_1
nodes [
 {id=Mixer_1, children [Random_1,Modulator_3], properties {inputVolume [1,1]}}
 {id=Modulator_3, children [Random_5], properties {pitchMax=1.2, pitchMin=1.1, volumeMax=1, volumeMin=0.9}}
 {id=Random_1, children [WavePlayer_3,WavePlayer_4,WavePlayer_5], properties {weights [1,1,1]}}
 {id=Random_5, children [WavePlayer_28, WavePlayer_29, WavePlayer_30, WavePlayer_31], properties {weights [1,1,1,1]}}
 {id=WavePlayer_28, properties {soundWaveAssetPtr=/ChronoSwarmGrenade/Sounds/Assets/InsectSwarm_Sting_HissLayer_01}}
 {id=WavePlayer_29, properties {soundWaveAssetPtr=/ChronoSwarmGrenade/Sounds/Assets/InsectSwarm_Sting_HissLayer_02}}
 {id=WavePlayer_3, properties {soundWaveAssetPtr=/ChronoSwarmGrenade/Sounds/Assets/InsectSwarm_Death_B_01}}
 {id=WavePlayer_30, properties {soundWaveAssetPtr=/ChronoSwarmGrenade/Sounds/Assets/InsectSwarm_Sting_HissLayer_03}}
 {id=WavePlayer_31, properties {soundWaveAssetPtr=/ChronoSwarmGrenade/Sounds/Assets/InsectSwarm_Sting_HissLayer_04}}
 {id=WavePlayer_4, properties {soundWaveAssetPtr=/ChronoSwarmGrenade/Sounds/Assets/InsectSwarm_Death_B_02}}
 {id=WavePlayer_5, properties {soundWaveAssetPtr=/ChronoSwarmGrenade/Sounds/Assets/InsectSwarm_Death_B_03}}
]
properties {
 attenuationSettings=/InsectSwarm/Sounds/InsectSwarm_Enemy_Close_Att
 bOverrideConcurrency=true
 maxDistance=3500
 soundClassObject=&Fort_WeaponRanged_Projectile_Effects
 volumeMultiplier=0.9
 attenuationOverrides {distanceAlgorithm=NaturalSound, falloffDistance=4000, attenuationShapeExtents {x=1000}}
 concurrencyOverrides {maxCount=3, resolutionRule=StopFarthestThenPreventNew}
}
