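# Sound cue for gathering the insect swarm jar, classed and attenuated as a Fortnite consumable sound.
# It randomly picks one of three equally weighted gather waves from the ChronoSwarmGrenade plugin and passes it through a modulator for slight upward pitch variation.
# A concurrency override allows one instance at a time, stopping the oldest when a new one starts.
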SoundCue /InsectSwarm/Sounds/Assets/InsectSwarm_Gather_Cue
root=Modulator_1
nodes [
 {id=Modulator_1, children [Random_0], properties {pitchMax=1.1, pitchMin=1}}
 {id=Random_0, children [WavePlayer_1,WavePlayer_2,WavePlayer_3], properties {weights [1,1,1]}}
 {id=WavePlayer_1, properties {soundWaveAssetPtr=/ChronoSwarmGrenade/Sounds/Assets/InsectSwarm_Jar_Gather_01}}
 {id=WavePlayer_2, properties {soundWaveAssetPtr=/ChronoSwarmGrenade/Sounds/Assets/InsectSwarm_Jar_Gather_02}}
 {id=WavePlayer_3, properties {soundWaveAssetPtr=/ChronoSwarmGrenade/Sounds/Assets/InsectSwarm_Jar_Gather_03}}
]
properties {
 attenuationSettings=&Fort_Consumable_Use_Athena_Att
 bOverrideConcurrency=true
 maxDistance=2250
 soundClassObject=&Fort_Consumables
 volumeMultiplier=1.2
 concurrencyOverrides {maxCount=1, resolutionRule=StopOldest}
}
