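# InsectSwarm hit sound cue that plays a layered insect sting, combining a bee layer with a hiss layer.
# A mixer blends a louder, randomly picked bee sample with a quieter random hiss sample that starts after a short fixed delay.
# Each layer has its own random pitch and volume modulation, with the bee layer varying more widely in pitch.
# It uses close-range enemy attenuation and overrides concurrency to cap simultaneous plays, stopping the farthest first.
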SoundCue /InsectSwarm/Sounds/Assets/InsectSwarm_Hit_Cue
root=Mixer_0
nodes [
 {id=Delay_0, children [Modulator_3], properties {delayMax=0.1, delayMin=0.1}}
 {id=Mixer_0, children [Modulator_2,Delay_0], properties {inputVolume [1.4,0.8]}}
 {id=Modulator_2, children [Random_2], properties {pitchMax=1.3, pitchMin=0.7, volumeMax=1.1, volumeMin=0.9}}
 {id=Modulator_3, children [Random_5], properties {pitchMax=1.2, pitchMin=1.1, volumeMax=1, volumeMin=0.9}}
 {
  id=Random_2
  children [WavePlayer_9, WavePlayer_10, WavePlayer_11, WavePlayer_12, WavePlayer_13, WavePlayer_14, WavePlayer_15, WavePlayer_16, WavePlayer_19]
  properties {weights [1,1,1,1,1,1,1,1,1]}
 }
 {id=Random_5, children [WavePlayer_28, WavePlayer_29, WavePlayer_30, WavePlayer_31], properties {weights [1,1,1,1]}}
 {id=WavePlayer_10, properties {soundWaveAssetPtr=/ChronoSwarmGrenade/Sounds/Assets/InsectSwarm_Sting_BeeLayer_02}}
 {id=WavePlayer_11, properties {soundWaveAssetPtr=/ChronoSwarmGrenade/Sounds/Assets/InsectSwarm_Sting_BeeLayer_03}}
 {id=WavePlayer_12, properties {soundWaveAssetPtr=/ChronoSwarmGrenade/Sounds/Assets/InsectSwarm_Sting_BeeLayer_04}}
 {id=WavePlayer_13, properties {soundWaveAssetPtr=/ChronoSwarmGrenade/Sounds/Assets/InsectSwarm_Sting_BeeLayer_05}}
 {id=WavePlayer_14, properties {soundWaveAssetPtr=/ChronoSwarmGrenade/Sounds/Assets/InsectSwarm_Sting_BeeLayer_06}}
 {id=WavePlayer_15, properties {soundWaveAssetPtr=/ChronoSwarmGrenade/Sounds/Assets/InsectSwarm_Sting_BeeLayer_07}}
 {id=WavePlayer_16, properties {soundWaveAssetPtr=/ChronoSwarmGrenade/Sounds/Assets/InsectSwarm_Sting_BeeLayer_08}}
 {id=WavePlayer_19, properties {soundWaveAssetPtr=/ChronoSwarmGrenade/Sounds/Assets/InsectSwarm_Sting_BeeLayer_09}}
 {id=WavePlayer_28, properties {soundWaveAssetPtr=/ChronoSwarmGrenade/Sounds/Assets/InsectSwarm_Sting_HissLayer_01}}
 {id=WavePlayer_29, properties {soundWaveAssetPtr=/ChronoSwarmGrenade/Sounds/Assets/InsectSwarm_Sting_HissLayer_02}}
 {id=WavePlayer_30, properties {soundWaveAssetPtr=/ChronoSwarmGrenade/Sounds/Assets/InsectSwarm_Sting_HissLayer_03}}
 {id=WavePlayer_31, properties {soundWaveAssetPtr=/ChronoSwarmGrenade/Sounds/Assets/InsectSwarm_Sting_HissLayer_04}}
 {id=WavePlayer_9, properties {soundWaveAssetPtr=/ChronoSwarmGrenade/Sounds/Assets/InsectSwarm_Sting_BeeLayer_01}}
]
properties {
 attenuationSettings=/InsectSwarm/Sounds/InsectSwarm_Enemy_Close_Att
 bOverrideConcurrency=true
 maxDistance=3500
 soundClassObject=&Fort_WeaponRanged_Projectile_Effects
 volumeMultiplier=0.8
 attenuationOverrides {distanceAlgorithm=NaturalSound, falloffDistance=4000, attenuationShapeExtents {x=1000}}
 concurrencyOverrides {maxCount=3, resolutionRule=StopFarthestThenPreventNew}
}
