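# Looping sound cue for an insect swarm that blends attack, ambient and plant-eating layers, limited to one instance per owner.
# The IsHostile parameter crossfades inversely between ambient loops and attack loops with a non-looping attack-start layer, while OnPlant scales the eating loops.
# Sparse and full variants use separate attenuation assets, and every layer picks random waves with pitch and volume modulation.
# The eating branch overrides attenuation and sound class.
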SoundCue /InsectSwarm/Sounds/Assets/InsectSwarm_Loop_Cue
root=Mixer_2
nodes [
 {id=Attenuation_0, children [Modulator_0], properties {attenuationSettings=/ChronoSwarmGrenade/Sounds/SwarmGrenade_Att}}
 {id=Attenuation_1, children [Modulator_1], properties {attenuationSettings=/ChronoSwarmGrenade/Sounds/SwarmGrenade_Sparse_Att}}
 {id=Attenuation_2, children [Modulator_2], properties {attenuationSettings=&SwarmGrenade_Sparse_Att}}
 {id=Attenuation_3, children [Modulator_3], properties {attenuationSettings=&SwarmGrenade_Att}}
 {id=Attenuation_4, children [Modulator_4], properties {attenuationSettings=&SwarmGrenade_Att}}
 {
  id=Attenuation_5
  children [SoundClass_0]
  properties {bOverrideAttenuation=true, attenuationOverrides {falloffDistance=1200, attenuationShapeExtents {x=100}}}
 }
 {id=Mixer_0, children [Attenuation_1, Attenuation_0, Attenuation_4], properties {inputVolume [0.75,0.85,0.65]}}
 {id=Mixer_1, children [Attenuation_2, Attenuation_3], properties {inputVolume [0.9,1.2]}}
 {id=Mixer_2, children [ModulatorContinuous_0, ModulatorContinuous_1, ModulatorContinuous_3], properties {inputVolume [0.9,1,0.6]}}
 {id=ModulatorContinuous_0, children [Mixer_0], properties {volumeModulationParams {default=0, parameterName=IsHostile}}}
 {
  id=ModulatorContinuous_1
  children [Mixer_1]
  properties {volumeModulationParams {default=0, maxOutput=0, minOutput=1, parameterName=IsHostile}}
 }
 {id=ModulatorContinuous_3, children [Attenuation_5], properties {volumeModulationParams {default=0, parameterName=OnPlant}}}
 {id=Modulator_0, children [Random_6], properties {pitchMax=1.1, pitchMin=0.9, volumeMax=1.1, volumeMin=0.9}}
 {id=Modulator_1, children [Random_5], properties {pitchMax=1.1, pitchMin=0.9, volumeMax=1.1, volumeMin=0.9}}
 {id=Modulator_2, children [Random_7], properties {pitchMax=1.1, pitchMin=0.9, volumeMax=1.1, volumeMin=0.9}}
 {id=Modulator_3, children [Random_8], properties {pitchMax=1.1, pitchMin=0.9, volumeMax=1.1, volumeMin=0.9}}
 {id=Modulator_4, children [Random_10], properties {pitchMax=1.2, pitchMin=1.1, volumeMax=1.1, volumeMin=0.9}}
 {id=Modulator_5, children [Random_0], properties {pitchMax=1.1}}
 {id=Random_0, children [WavePlayer_0,WavePlayer_1,WavePlayer_2], properties {weights [1,1,1]}}
 {id=Random_10, children [WavePlayer_41, WavePlayer_42, WavePlayer_43], properties {weights [1,1,1]}}
 {id=Random_5, children [WavePlayer_18, WavePlayer_19, WavePlayer_20], properties {weights [1,1,1]}}
 {id=Random_6, children [WavePlayer_21, WavePlayer_22, WavePlayer_23, WavePlayer_24], properties {weights [1,1,1,1]}}
 {
  id=Random_7
  children [WavePlayer_25, WavePlayer_26, WavePlayer_27, WavePlayer_28, WavePlayer_29, WavePlayer_30]
  properties {weights [1,1,1,1,1,1]}
 }
 {
  id=Random_8
  children [WavePlayer_31, WavePlayer_32, WavePlayer_33, WavePlayer_34, WavePlayer_35, WavePlayer_36]
  properties {weights [1,1,1,1,1,1]}
 }
 {id=SoundClass_0, children [Modulator_5], properties {soundClassOverride=&Fort_Gadgets_Items}}
 {id=WavePlayer_0, properties {bLooping=true, soundWaveAssetPtr=/ChronoSwarmGrenade/Sounds/Assets/InsectSwarm_PlatntLoop_Eating_01}}
 {id=WavePlayer_1, properties {bLooping=true, soundWaveAssetPtr=/ChronoSwarmGrenade/Sounds/Assets/InsectSwarm_PlatntLoop_Eating_02}}
 {id=WavePlayer_18, properties {bLooping=true, soundWaveAssetPtr=/ChronoSwarmGrenade/Sounds/Assets/InsectSwarm_AttackLoop_Sparse_B_01}}
 {id=WavePlayer_19, properties {bLooping=true, soundWaveAssetPtr=/ChronoSwarmGrenade/Sounds/Assets/InsectSwarm_AttackLoop_Sparse_B_02}}
 {id=WavePlayer_2, properties {bLooping=true, soundWaveAssetPtr=/ChronoSwarmGrenade/Sounds/Assets/InsectSwarm_PlatntLoop_Eating_03}}
 {id=WavePlayer_20, properties {bLooping=true, soundWaveAssetPtr=/ChronoSwarmGrenade/Sounds/Assets/InsectSwarm_AttackLoop_Sparse_B_03}}
 {id=WavePlayer_21, properties {bLooping=true, soundWaveAssetPtr=/ChronoSwarmGrenade/Sounds/Assets/InsectSwarm_AttackLoop_Full_B_01}}
 {id=WavePlayer_22, properties {bLooping=true, soundWaveAssetPtr=/ChronoSwarmGrenade/Sounds/Assets/InsectSwarm_AttackLoop_Full_B_02}}
 {id=WavePlayer_23, properties {bLooping=true, soundWaveAssetPtr=/ChronoSwarmGrenade/Sounds/Assets/InsectSwarm_AttackLoop_Full_B_03}}
 {id=WavePlayer_24, properties {bLooping=true, soundWaveAssetPtr=/ChronoSwarmGrenade/Sounds/Assets/InsectSwarm_AttackLoop_Full_B_04}}
 {id=WavePlayer_25, properties {bLooping=true, soundWaveAssetPtr=/ChronoSwarmGrenade/Sounds/Assets/InsectSwarm_AmbientLoop_Sparse_B_02}}
 {id=WavePlayer_26, properties {bLooping=true, soundWaveAssetPtr=/ChronoSwarmGrenade/Sounds/Assets/InsectSwarm_AmbientLoop_Sparse_B_03}}
 {id=WavePlayer_27, properties {bLooping=true, soundWaveAssetPtr=/ChronoSwarmGrenade/Sounds/Assets/InsectSwarm_AmbientLoop_Sparse_B_04}}
 {id=WavePlayer_28, properties {bLooping=true, soundWaveAssetPtr=/ChronoSwarmGrenade/Sounds/Assets/InsectSwarm_AmbientLoop_Sparse_C_01}}
 {id=WavePlayer_29, properties {bLooping=true, soundWaveAssetPtr=/ChronoSwarmGrenade/Sounds/Assets/InsectSwarm_AmbientLoop_Sparse_C_02}}
 {id=WavePlayer_30, properties {bLooping=true, soundWaveAssetPtr=/ChronoSwarmGrenade/Sounds/Assets/InsectSwarm_AmbientLoop_Sparse_C_03}}
 {id=WavePlayer_31, properties {bLooping=true, soundWaveAssetPtr=/ChronoSwarmGrenade/Sounds/Assets/InsectSwarm_AmbientLoop_Full_03}}
 {id=WavePlayer_32, properties {bLooping=true, soundWaveAssetPtr=/ChronoSwarmGrenade/Sounds/Assets/InsectSwarm_AmbientLoop_Full_04}}
 {id=WavePlayer_33, properties {bLooping=true, soundWaveAssetPtr=/ChronoSwarmGrenade/Sounds/Assets/InsectSwarm_AmbientLoop_Full_B_01}}
 {id=WavePlayer_34, properties {bLooping=true, soundWaveAssetPtr=/ChronoSwarmGrenade/Sounds/Assets/InsectSwarm_AmbientLoop_Full_B_02}}
 {id=WavePlayer_35, properties {bLooping=true, soundWaveAssetPtr=/ChronoSwarmGrenade/Sounds/Assets/InsectSwarm_AmbientLoop_Full_C_01}}
 {id=WavePlayer_36, properties {bLooping=true, soundWaveAssetPtr=/ChronoSwarmGrenade/Sounds/Assets/InsectSwarm_AmbientLoop_Full_C_02}}
 {id=WavePlayer_41, properties {soundWaveAssetPtr=/ChronoSwarmGrenade/Sounds/Assets/InsectSwarm_AtackStart_FastFade_01}}
 {id=WavePlayer_42, properties {soundWaveAssetPtr=/ChronoSwarmGrenade/Sounds/Assets/InsectSwarm_AtackStart_FastFade_02}}
 {id=WavePlayer_43, properties {soundWaveAssetPtr=/ChronoSwarmGrenade/Sounds/Assets/InsectSwarm_AtackStart_FastFade_03}}
]
properties {
 bOverrideConcurrency=true
 maxDistance=5278.5361
 pitchMultiplier=0.95
 soundClassObject=&Fort_WeaponRanged_Fire_3P_PSM
 volumeMultiplier=0.6
 concurrencyOverrides {bLimitToOwner=true, maxCount=1}
 soundSubmixSends [{minSendLevel=0, soundSubmix=&Fortnite_AudioAnalysis_Music}]
}
