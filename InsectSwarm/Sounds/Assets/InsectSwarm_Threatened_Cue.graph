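# Sound cue for the insect swarm's threatened sound, built from attack-start fast-fade waves in the ChronoSwarmGrenade content.
# An equally weighted random pick of three waves passes through a default Modulator, then an Enveloper that ramps volume up from near silence.
# It uses the swarm grenade attenuation and third-person ranged weapon fire class, allows one instance per owner, and sends to a music audio-analysis submix.
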SoundCue /InsectSwarm/Sounds/Assets/InsectSwarm_Threatened_Cue
root=Enveloper_0
nodes [
 {
  id=Enveloper_0
  children [Modulator_0]
  properties {pitchCurve {editorCurveData {keys [{time=0}]}}, volumeCurve {editorCurveData {keys [<at 0 0.01>,<at 0.62 0.9934>]}}}
 }
 {id=Modulator_0, children [Random_10]}
 {id=Random_10, children [WavePlayer_41, WavePlayer_42, WavePlayer_43], properties {weights [1,1,1]}}
 {id=WavePlayer_41, properties {soundWaveAssetPtr=/ChronoSwarmGrenade/Sounds/Assets/InsectSwarm_AtackStart_FastFade_01}}
 {id=WavePlayer_42, properties {soundWaveAssetPtr=/ChronoSwarmGrenade/Sounds/Assets/InsectSwarm_AtackStart_FastFade_02}}
 {id=WavePlayer_43, properties {soundWaveAssetPtr=/ChronoSwarmGrenade/Sounds/Assets/InsectSwarm_AtackStart_FastFade_03}}
]
properties {
 attenuationSettings=/ChronoSwarmGrenade/Sounds/SwarmGrenade_Att
 bOverrideConcurrency=true
 maxDistance=3000
 soundClassObject=&Fort_WeaponRanged_Fire_3P_PSM
 volumeMultiplier=0.45
 concurrencyOverrides {bLimitToOwner=true, maxCount=1}
 soundSubmixSends [{minSendLevel=0, soundSubmix=&Fortnite_AudioAnalysis_Music}]
}
